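# Physics asset for the male medium Dark Eagle Purple player body cosmetic, covering its dangling loincloth, arm wrap and boot strap bones.
# Capsules on pelvis, thighs, feet and upper arms form a coarse body, while small boxes cover the dyn_ bones.
# Tightly limited constraints chain six loincloth segments from the pelvis, top and bottom strands from each arm wrap base, and straps from each foot.
# Arm wrap bases rotate freely on all axes against the upper arms.
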PhysicsAsset /BRCosmetics/Characters/Player/Male/Medium/Bodies/M_MED_Dark_Eagle_Purple/Meshes/M_MED_Dark_Eagle_Purple_Physics
bodies {
 dyn_armwrap_l_base [{shapeType=Box, extentX=1.3417, extentY=1.3417, extentZ=1.3417}]
 dyn_armwrap_l_bot_01 [{shapeType=Box, extentX=2, extentY=2, extentZ=1.0638, center=<1.3292 0 0>}]
 dyn_armwrap_l_bot_02 [{shapeType=Box, extentX=2, extentY=2, extentZ=0.574, center=<1.1698 0.0001 0>}]
 dyn_armwrap_l_bot_03 [{shapeType=Box, extentX=1.1431, extentY=1.7772, extentZ=1.1431, center=<0.6955 0 0>}]
 dyn_armwrap_l_top_01 [{shapeType=Box, extentX=2.9417, extentY=2.5644, extentZ=1.1318, center=<1.5756 0 0>}]
 dyn_armwrap_l_top_02 [{shapeType=Box, extentX=2.946, extentY=2.9857, extentZ=1.0579, center=<1.7507 0.0001 0>}]
 dyn_armwrap_l_top_03 [{shapeType=Box, extentX=1.3758, extentY=2, extentZ=0.9583, center=<0.7261 0.0001 0>}]
 dyn_armwrap_r_base [{shapeType=Box, extentX=1.5748, extentY=1.5748, extentZ=1.5748}]
 dyn_armwrap_r_bot_01 [{shapeType=Box, extentX=2, extentY=2, extentZ=1.0638, center=<1.3292 0 0>}]
 dyn_armwrap_r_bot_02 [{shapeType=Box, extentX=2, extentY=2, extentZ=0.574, center=<1.1698 0.0001 0>}]
 dyn_armwrap_r_bot_03 [{shapeType=Box, extentX=1.1431, extentY=1.7772, extentZ=1.1431, center=<0.6955 0 0>}]
 dyn_armwrap_r_top_01 [{shapeType=Box, extentX=2.9417, extentY=2.5644, extentZ=1.1318, center=<1.5756 0 0>}]
 dyn_armwrap_r_top_02 [{shapeType=Box, extentX=2.946, extentY=2.9857, extentZ=1.0579, center=<1.7507 0.0001 0>}]
 dyn_armwrap_r_top_03 [{shapeType=Box, extentX=1.3758, extentY=2, extentZ=0.9583, center=<0.7261 0.0001 0>}]
 dyn_boot_strap_l [{shapeType=Box, extentX=2, extentY=3.1301, extentZ=0.7087, center=<1.1368 0 0>}]
 dyn_boot_strap_r [{shapeType=Box, extentX=2, extentY=3.2037, extentZ=0.9416, center=<0.9533 0 0>}]
 dyn_loincloth_01 [{shapeType=Box, extentX=10.4846, extentY=9.1503, extentZ=1.7372, center=<5.6873 0.0001 0>}]
 dyn_loincloth_02 [{shapeType=Box, extentX=7.2617, extentY=9.0879, extentZ=1.978, center=<3.5617 0 0>}]
 dyn_loincloth_03 [{shapeType=Box, extentX=8.3826, extentY=8.629, extentZ=1.978, center=<4.1677 0.0462 0>}]
 dyn_loincloth_04 [{shapeType=Box, extentX=8.5414, extentY=7.7875, extentZ=1.978, center=<4.3285 0.0253 0>}]
 dyn_loincloth_05 [{shapeType=Box, extentX=9.0422, extentY=6.5872, extentZ=1.978, center=<4.6713 0.0253 0>}]
 dyn_loincloth_06 [{shapeType=Box, extentX=7.6021, extentY=6.5872, extentZ=1.978, center=<3.5616 0.0253 0>}]
 foot_l [{shapeType=Capsule, length=5.1083, radius=4.8917, rotation=<rot 0 0 90>}]
 foot_r [{shapeType=Capsule, length=5, radius=5, rotation=<rot 0 0 90>}]
 pelvis [{shapeType=Capsule, length=19, radius=8, center=<-0.7566 0 0>}]
 thigh_l [{shapeType=Capsule, length=24.8294, radius=8.1706, center=<-18.9788 0.5247 0>, rotation=<rot 0 90 90>}]
 thigh_r [{shapeType=Capsule, length=23.9408, radius=8.3092, center=<18.8966 -0.6596 0>, rotation=<rot 0 90 90>}]
 upperarm_l [{shapeType=Capsule, length=15.25, radius=6, center=<13.9899 0.0001 0>, rotation=<rot 90 0 0>}]
 upperarm_r [{shapeType=Capsule, length=16, radius=6.5, center=<-13.2864 -0.0005 0>, rotation=<rot 90 0 0>}]
}
constraints [
 {bone1Name=upperarm_l, bone2Name=pelvis, swing1LimitDegrees=45, swing2LimitDegrees=45, twistLimitDegrees=45}
 {bone1Name=upperarm_r, bone2Name=pelvis, swing1LimitDegrees=45, swing2LimitDegrees=45, twistLimitDegrees=45}
 {bone1Name=dyn_loincloth_01, bone2Name=pelvis, swing1LimitDegrees=10, swing2LimitDegrees=20, twistLimitDegrees=2}
 {bone1Name=thigh_l, bone2Name=pelvis, swing1LimitDegrees=45, swing2LimitDegrees=45, twistLimitDegrees=45}
 {bone1Name=foot_l, bone2Name=thigh_l, swing1LimitDegrees=45, swing2LimitDegrees=45, twistLimitDegrees=45}
 {bone1Name=thigh_r, bone2Name=pelvis, swing1LimitDegrees=45, swing2LimitDegrees=45, twistLimitDegrees=45}
 {bone1Name=foot_r, bone2Name=thigh_r, swing1LimitDegrees=45, swing2LimitDegrees=45, twistLimitDegrees=45}
 {bone1Name=dyn_loincloth_02, bone2Name=dyn_loincloth_01, swing1LimitDegrees=15, swing2LimitDegrees=20, twistLimitDegrees=5}
 {bone1Name=dyn_loincloth_03, bone2Name=dyn_loincloth_02, swing1LimitDegrees=15, swing2LimitDegrees=20, twistLimitDegrees=5}
 {bone1Name=dyn_loincloth_04, bone2Name=dyn_loincloth_03, swing1LimitDegrees=15, swing2LimitDegrees=20, twistLimitDegrees=5}
 {bone1Name=dyn_loincloth_05, bone2Name=dyn_loincloth_04, swing1LimitDegrees=15, swing2LimitDegrees=20, twistLimitDegrees=5}
 {bone1Name=dyn_loincloth_06, bone2Name=dyn_loincloth_05, swing1LimitDegrees=15, swing2LimitDegrees=20, twistLimitDegrees=5}
 {bone1Name=dyn_armwrap_r_bot_03, bone2Name=dyn_armwrap_r_bot_02, swing1LimitDegrees=15, swing2LimitDegrees=30, twistLimitDegrees=5}
 {bone1Name=dyn_armwrap_r_bot_01, bone2Name=dyn_armwrap_r_base, swing1LimitDegrees=15, swing2LimitDegrees=30, twistLimitDegrees=5}
 {bone1Name=dyn_armwrap_r_bot_02, bone2Name=dyn_armwrap_r_bot_01, swing1LimitDegrees=15, swing2LimitDegrees=30, twistLimitDegrees=5}
 {bone1Name=dyn_armwrap_r_top_03, bone2Name=dyn_armwrap_r_top_02, swing1LimitDegrees=15, swing2LimitDegrees=30, twistLimitDegrees=5}
 {bone1Name=dyn_armwrap_r_top_01, bone2Name=dyn_armwrap_r_base, swing1LimitDegrees=15, swing2LimitDegrees=30, twistLimitDegrees=5}
 {bone1Name=dyn_armwrap_r_top_02, bone2Name=dyn_armwrap_r_top_01, swing1LimitDegrees=15, swing2LimitDegrees=30, twistLimitDegrees=5}
 {bone1Name=dyn_armwrap_l_bot_03, bone2Name=dyn_armwrap_l_bot_02, swing1LimitDegrees=15, swing2LimitDegrees=30, twistLimitDegrees=5}
 {bone1Name=dyn_armwrap_l_bot_02, bone2Name=dyn_armwrap_l_bot_01, swing1LimitDegrees=15, swing2LimitDegrees=30, twistLimitDegrees=5}
 {bone1Name=dyn_armwrap_l_bot_01, bone2Name=dyn_armwrap_l_base, swing1LimitDegrees=15, swing2LimitDegrees=30, twistLimitDegrees=5}
 {bone1Name=dyn_armwrap_l_top_03, bone2Name=dyn_armwrap_l_top_02, swing1LimitDegrees=15, swing2LimitDegrees=30, twistLimitDegrees=5}
 {bone1Name=dyn_armwrap_l_top_01, bone2Name=dyn_armwrap_l_base, swing1LimitDegrees=15, swing2LimitDegrees=30, twistLimitDegrees=5}
 {bone1Name=dyn_armwrap_l_top_02, bone2Name=dyn_armwrap_l_top_01, swing1LimitDegrees=15, swing2LimitDegrees=30, twistLimitDegrees=5}
 {
  bone1Name=dyn_armwrap_l_base
  bone2Name=upperarm_l
  swing1LimitDegrees=45
  swing1Motion=Free
  swing2LimitDegrees=45
  swing2Motion=Free
  twistLimitDegrees=45
  twistMotion=Free
 }
 {
  bone1Name=dyn_armwrap_r_base
  bone2Name=upperarm_r
  swing1LimitDegrees=45
  swing1Motion=Free
  swing2LimitDegrees=45
  swing2Motion=Free
  twistLimitDegrees=45
  twistMotion=Free
 }
 {bone1Name=dyn_boot_strap_r, bone2Name=foot_r, swing1LimitDegrees=10, swing2LimitDegrees=20, twistLimitDegrees=3}
 {bone1Name=dyn_boot_strap_l, bone2Name=foot_l, swing1LimitDegrees=10, swing2LimitDegrees=20, twistLimitDegrees=3}
]
