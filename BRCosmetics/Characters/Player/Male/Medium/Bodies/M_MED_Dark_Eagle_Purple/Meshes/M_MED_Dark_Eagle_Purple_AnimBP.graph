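# Post-process style cosmetic AnimBP for the male medium Dark Eagle Purple body, adding secondary motion to the pose copied from its attached parent mesh.
# RigidBody physics, weighted by sprint state, feeds armor bone constraints, then additive pad and forearm corrections driven by upperarm-to-clavicle yaw (LookUp/LookDown).
# The cached result gets a right forearm offset chosen by targeting state, then chained AnimDynamics springs jiggle both bag bone chains.
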AnimBlueprint /BRCosmetics/Characters/Player/Male/Medium/Bodies/M_MED_Dark_Eagle_Purple/Meshes/M_MED_Dark_Eagle_Purple_AnimBP
target_skeleton=&/M_MED_Dark_Eagle_Purple_Skeleton
graphs [
 AnimGraph {nodes [
   RigidBody() {
    overridePhysicsAsset=&/M_MED_Dark_Eagle_Purple_Physics
    alphaBoolBlend {blendOption=Linear}
   }
   n1 = "Transform (Modify) Bone - Bone: dyn_forearm_armor_r"(
    ComponentPose=Local To Component(LocalPose=Use cached pose 'Before_R_Forearm'())
    Alpha=1
    Translation="0,0,0"
    Rotation="0,5,0"
    Scale="1,1,1"
   ) {
    rotationMode=BMM_Additive
    rotationSpace=BCS_ParentBoneSpace
    scaleMode=BMM_Additive
    translationMode=BMM_Additive
    alphaBoolBlend {blendOption=Linear}
    boneToModify {boneName=dyn_forearm_armor_r}
    rotation {yaw=5}
   }
   n2 = Component To Local(ComponentPose=@n1)
   n3 = Use cached pose 'Before_R_Forearm'()
   n4 = Get bIsTargeting()
   n5 = Anim Dynamics - Bone: dyn_bag_l_01(
    ComponentPose=Local To Component(LocalPose=Blend Poses by bool(BlendPose_0=@n2, BlendPose_1=@n3, bActiveValue=@n4) {blendType=Linear})
    Alpha=1
    ExternalForce="0,0,0"
   ) {
    angularDampingOverride=0.999
    angularSpringConstant=1000
    bAngularSpring=true
    bGravityOverrideInSimSpace=true
    bLinearSpring=true
    bOverrideAngularDamping=true
    bOverrideLinearDamping=true
    bUseGravityOverride=true
    gravityScale=0.3
    linearDampingOverride=0.999
    linearSpringConstant=200
    simulationSpace=BoneRelative
    alphaBoolBlend {blendOption=Linear}
    boundBone {boneName=dyn_bag_l_01}
    componentAppliedLinearAccClamp=<1000 1000 1000>
    componentLinearAccScale=<1 1 1>
    componentLinearVelScale=<1 1 1>
    gravityOverride {z=-700}
    physicsBodyDefinitions [{
      sphereCollisionRadius=0
      boundBone {boneName=dyn_bag_l_01}
      boxExtents=<5 10 10>
      constraintSetup {
       angularTargetAxis=AxisY
       angularLimitsMax=<2 2 2>
       angularLimitsMin=<-2 -2 -2>
       angularTarget=<0 0 0>
       linearAxesMax=<0.1 0.1 0.1>
       linearAxesMin=<-0.1 -0.1 -0.1>
      }
      localJointOffset=<1 0 0>
     }]
   }
   n6 = Anim Dynamics - Bone: dyn_bag_l_02(ComponentPose=@n5, Alpha=0.9, ExternalForce="0,0,0") {
    angularDampingOverride=0.999
    angularSpringConstant=1000
    bAngularSpring=true
    bGravityOverrideInSimSpace=true
    bLinearSpring=true
    bOverrideAngularDamping=true
    bOverrideLinearDamping=true
    bUseGravityOverride=true
    gravityScale=0.3
    linearDampingOverride=0.999
    linearSpringConstant=200
    simulationSpace=BoneRelative
    alphaBoolBlend {blendOption=Linear}
    boundBone {boneName=dyn_bag_l_02}
    componentAppliedLinearAccClamp=<1000 1000 1000>
    componentLinearAccScale=<1 1 1>
    componentLinearVelScale=<1 1 1>
    gravityOverride {z=-700}
    physicsBodyDefinitions [{
      sphereCollisionRadius=0
      boundBone {boneName=dyn_bag_l_02}
      boxExtents=<5 10 10>
      constraintSetup {
       angularTargetAxis=AxisY
       angularLimitsMax=<2 2 2>
       angularLimitsMin=<-2 -2 -2>
       angularTarget=<0 0 0>
       linearAxesMax=<0.1 0.1 0.1>
       linearAxesMin=<-0.1 -0.1 -0.1>
      }
      localJointOffset=<1 0 0>
     }]
   }
   n7 = Anim Dynamics - Bone: dyn_bag_l_03(ComponentPose=@n6, Alpha=0.8, ExternalForce="0,0,0") {
    angularDampingOverride=0.999
    angularSpringConstant=1000
    bAngularSpring=true
    bGravityOverrideInSimSpace=true
    bLinearSpring=true
    bOverrideAngularDamping=true
    bOverrideLinearDamping=true
    bUseGravityOverride=true
    gravityScale=0.3
    linearDampingOverride=0.999
    linearSpringConstant=200
    simulationSpace=BoneRelative
    alphaBoolBlend {blendOption=Linear}
    boundBone {boneName=dyn_bag_l_03}
    componentAppliedLinearAccClamp=<1000 1000 1000>
    componentLinearAccScale=<1 1 1>
    componentLinearVelScale=<1 1 1>
    gravityOverride {z=-700}
    physicsBodyDefinitions [{
      sphereCollisionRadius=0
      boundBone {boneName=dyn_bag_l_03}
      boxExtents=<5 10 10>
      constraintSetup {
       angularTargetAxis=AxisY
       angularLimitsMax=<2 2 2>
       angularLimitsMin=<-2 -2 -2>
       angularTarget=<0 0 0>
       linearAxesMax=<0.1 0.1 0.1>
       linearAxesMin=<-0.1 -0.1 -0.1>
      }
      localJointOffset=<1 0 0>
     }]
   }
   n8 = Anim Dynamics - Bone: dyn_bag_r_01(ComponentPose=@n7, Alpha=1, ExternalForce="0,0,0") {
    angularDampingOverride=0.999
    angularSpringConstant=1000
    bAngularSpring=true
    bGravityOverrideInSimSpace=true
    bLinearSpring=true
    bOverrideAngularDamping=true
    bOverrideLinearDamping=true
    bUseGravityOverride=true
    gravityScale=0.3
    linearDampingOverride=0.999
    linearSpringConstant=200
    simulationSpace=BoneRelative
    alphaBoolBlend {blendOption=Linear}
    boundBone {boneName=dyn_bag_r_01}
    componentAppliedLinearAccClamp=<1000 1000 1000>
    componentLinearAccScale=<1 1 1>
    componentLinearVelScale=<1 1 1>
    gravityOverride {z=-700}
    physicsBodyDefinitions [{
      sphereCollisionRadius=0
      boundBone {boneName=dyn_bag_r_01}
      boxExtents=<5 10 10>
      constraintSetup {
       angularTargetAxis=AxisY
       angularLimitsMax=<2 2 2>
       angularLimitsMin=<-2 -2 -2>
       angularTarget=<0 0 0>
       linearAxesMax=<0.1 0.1 0.1>
       linearAxesMin=<-0.1 -0.1 -0.1>
      }
      localJointOffset=<1 0 0>
     }]
   }
   n9 = Anim Dynamics - Bone: dyn_bag_r_02(ComponentPose=@n8, Alpha=0.9, ExternalForce="0,0,0") {
    angularDampingOverride=0.999
    angularSpringConstant=1000
    bAngularSpring=true
    bGravityOverrideInSimSpace=true
    bLinearSpring=true
    bOverrideAngularDamping=true
    bOverrideLinearDamping=true
    bUseGravityOverride=true
    gravityScale=0.3
    linearDampingOverride=0.999
    linearSpringConstant=200
    simulationSpace=BoneRelative
    alphaBoolBlend {blendOption=Linear}
    boundBone {boneName=dyn_bag_r_02}
    componentAppliedLinearAccClamp=<1000 1000 1000>
    componentLinearAccScale=<1 1 1>
    componentLinearVelScale=<1 1 1>
    gravityOverride {z=-700}
    physicsBodyDefinitions [{
      sphereCollisionRadius=0
      boundBone {boneName=dyn_bag_r_02}
      boxExtents=<5 10 10>
      constraintSetup {
       angularTargetAxis=AxisY
       angularLimitsMax=<2 2 2>
       angularLimitsMin=<-2 -2 -2>
       angularTarget=<0 0 0>
       linearAxesMax=<0.1 0.1 0.1>
       linearAxesMin=<-0.1 -0.1 -0.1>
      }
      localJointOffset=<1 0 0>
     }]
   }
   n10 = Anim Dynamics - Bone: dyn_bag_r_03(ComponentPose=@n9, Alpha=0.8, ExternalForce="0,0,0") {
    angularDampingOverride=0.999
    angularSpringConstant=1000
    bAngularSpring=true
    bGravityOverrideInSimSpace=true
    bLinearSpring=true
    bOverrideAngularDamping=true
    bOverrideLinearDamping=true
    bUseGravityOverride=true
    gravityScale=0.3
    linearDampingOverride=0.999
    linearSpringConstant=200
    simulationSpace=BoneRelative
    alphaBoolBlend {blendOption=Linear}
    boundBone {boneName=dyn_bag_r_03}
    componentAppliedLinearAccClamp=<1000 1000 1000>
    componentLinearAccScale=<1 1 1>
    componentLinearVelScale=<1 1 1>
    gravityOverride {z=-700}
    physicsBodyDefinitions [{
      sphereCollisionRadius=0
      boundBone {boneName=dyn_bag_r_03}
      boxExtents=<5 10 10>
      constraintSetup {
       angularTargetAxis=AxisY
       angularLimitsMax=<2 2 2>
       angularLimitsMin=<-2 -2 -2>
       angularTarget=<0 0 0>
       linearAxesMax=<0.1 0.1 0.1>
       linearAxesMin=<-0.1 -0.1 -0.1>
      }
      localJointOffset=<1 0 0>
     }]
   }
   Output Pose(Result=Component To Local(ComponentPose=@n10)) {layerGroup=None}
   n11 = RigidBody(ComponentPose=Local To Component(LocalPose=Copy Pose From Mesh() {bUseAttachedParent=true}), Alpha=0.9) {
    bClampLinearTranslationLimitToRefPose=true
    bOverrideWorldGravity=true
    overridePhysicsAsset=&M_MED_Dark_Eagle_Purple_Physics
    simulationSpace=BaseBoneSpace
    alphaBoolBlend {blendOption=Linear}
    baseBoneRef {boneName=root}
    componentLinearAccScale=<0.5 0.5 0.5>
    overrideWorldGravity {z=-980}
   }
   n12 = Component To Local(ComponentPose=@n11)
   n13 = RigidBody(ComponentPose=Local To Component(LocalPose=Copy Pose From Mesh() {bUseAttachedParent=true}), Alpha=0.7) {
    bClampLinearTranslationLimitToRefPose=true
    bOverrideWorldGravity=true
    overridePhysicsAsset=&M_MED_Dark_Eagle_Purple_Physics
    simulationSpace=BaseBoneSpace
    alphaBoolBlend {blendOption=Linear}
    baseBoneRef {boneName=root}
    componentLinearAccScale=<0.5 0.5 0.5>
    overrideWorldGravity {z=-980}
   }
   n14 = Component To Local(ComponentPose=@n13)
   n15 = Get bIsSprinting()
   n16 = "Constraint - dyn_forearm_armor_l (PPPPPP)"!(
    ComponentPose=Local To Component(LocalPose=Blend Poses by bool(BlendPose_0=@n14, BlendPose_1=@n12, bActiveValue=@n15) {blendType=Linear})
    Alpha=1
    ConstraintWeights_0=0.013
    ConstraintWeights_1:float=0.2268
    ConstraintWeights_2:float=0.0001
    ConstraintWeights_3:float=0.4667
    ConstraintWeights_4:float=0.0029
    ConstraintWeights_5:float=0.2905
   ) -> (Pose) {
    alphaBoolBlend {blendOption=Linear}
    boneToModify {boneName=dyn_forearm_armor_l}
    constraintSetup [
     {targetBone {boneName=clavicle_l}}
     {targetBone {boneName=upperarm_twist_02_l}}
     {targetBone {boneName=spine_03}}
     {targetBone {boneName=upperarm_twist_01_l}}
     {targetBone {boneName=lowerarm_l}}
     {targetBone {boneName=upperarm_l}}
    ]
    constraintWeights [1,1,1,1,1,1]
   }
   n17 = "Constraint - dyn_forearm_armor_r (PPPPPPP)"!(
    ComponentPose=@n16
    Alpha=1
    ConstraintWeights_0=0
    ConstraintWeights_1:float=0.2808
    ConstraintWeights_2:float=0.0039
    ConstraintWeights_3:float=0.4801
    ConstraintWeights_4:float=0.2191
    ConstraintWeights_5:float=0.0004
    ConstraintWeights_6:float=0.0157
   ) -> (Pose) {
    alphaBoolBlend {blendOption=Linear}
    boneToModify {boneName=dyn_forearm_armor_r}
    constraintSetup [
     {targetBone {boneName=lowerarm_twist_02_r}}
     {targetBone {boneName=upperarm_r}}
     {targetBone {boneName=lowerarm_r}}
     {targetBone {boneName=upperarm_twist_01_r}}
     {targetBone {boneName=upperarm_twist_02_r}}
     {targetBone {boneName=spine_03}}
     {targetBone {boneName=clavicle_r}}
    ]
    constraintWeights [1,1,1,1,1,1,1]
   }
   n18 = "Constraint - dyn_shoulder_pad_r (PPPPPP)"!(
    ComponentPose=@n17
    Alpha=1
    ConstraintWeights_0=0.0083
    ConstraintWeights_1:float=0.7414
    ConstraintWeights_2:float=0.0079
    ConstraintWeights_3:float=0.0189
    ConstraintWeights_4:float=0.0099
    ConstraintWeights_5:float=0.2136
   ) -> (Pose) {
    alphaBoolBlend {blendOption=Linear}
    boneToModify {boneName=dyn_shoulder_pad_r}
    constraintSetup [
     {targetBone {boneName=upperarm_r}}
     {targetBone {boneName=upperarm_twist_01_r}}
     {targetBone {boneName=upperarm_twist_02_r}}
     {targetBone {boneName=spine_03}}
     {targetBone {boneName=spine_05}}
     {targetBone {boneName=clavicle_r}}
    ]
    constraintWeights [1,1,1,1,1,1]
   }
   n19 = "Constraint - dyn_shoulder_pad_l (PPPPPPPP)"!(
    ComponentPose=@n18
    Alpha=1
    ConstraintWeights_0=0.29
    ConstraintWeights_1:float=0.0305
    ConstraintWeights_2:float=0
    ConstraintWeights_3:float=0.0263
    ConstraintWeights_4:float=0.0012
    ConstraintWeights_5:float=0.0183
    ConstraintWeights_6:float=0.6045
    ConstraintWeights_7:float=0.0292
   ) -> (Pose) {
    alphaBoolBlend {blendOption=Linear}
    boneToModify {boneName=dyn_shoulder_pad_l}
    constraintSetup [
     {targetBone {boneName=clavicle_l}}
     {targetBone {boneName=upperarm_twist_02_l}}
     {targetBone {boneName=spine_02}}
     {targetBone {boneName=spine_03}}
     {targetBone {boneName=spine_04}}
     {targetBone {boneName=spine_05}}
     {targetBone {boneName=upperarm_twist_01_l}}
     {targetBone {boneName=upperarm_l}}
    ]
    constraintWeights [1,1,1,1,1,1,1,1]
   }
   n20 = Get LookUp() -> (LookUp:double)
   n21 = "Transform (Modify) Bone - Bone: dyn_shoulder_pad_r"(ComponentPose=@n19, Alpha=@n20, Translation="-1,0,1", Rotation="0,0,-30", Scale="1,1,1") {
    rotationMode=BMM_Additive
    rotationSpace=BCS_ParentBoneSpace
    scaleMode=BMM_Additive
    translationMode=BMM_Additive
    alphaBoolBlend {blendOption=Linear}
    boneToModify {boneName=dyn_shoulder_pad_r}
    rotation {roll=-30}
    translation {x=-1, z=1}
   }
   n22 = "Transform (Modify) Bone - Bone: dyn_shoulder_pad_l"(ComponentPose=@n21, Alpha=@n20, Translation="0,0,0", Rotation="0,20,-20", Scale="1,1,1") {
    rotationMode=BMM_Additive
    rotationSpace=BCS_ParentBoneSpace
    scaleMode=BMM_Additive
    translationMode=BMM_Additive
    alphaBoolBlend {blendOption=Linear}
    boneToModify {boneName=dyn_shoulder_pad_l}
    rotation {roll=-20, yaw=20}
   }
   n23 = "Transform (Modify) Bone - Bone: dyn_forearm_armor_l"(ComponentPose=@n22, Alpha=@n20, Translation="0,0,0", Rotation="0,10,0", Scale="1,1,1") {
    rotationMode=BMM_Additive
    rotationSpace=BCS_ParentBoneSpace
    scaleMode=BMM_Additive
    translationMode=BMM_Additive
    alphaBoolBlend {blendOption=Linear}
    boneToModify {boneName=dyn_forearm_armor_l}
    rotation {yaw=10}
   }
   n24 = "Transform (Modify) Bone - Bone: dyn_shoulder_pad_r"(ComponentPose=@n23, Alpha=Get LookDown() -> (LookDown:double), Translation="0,0,0", Rotation="0,0,-10", Scale="1,1,1") {
    rotationMode=BMM_Additive
    rotationSpace=BCS_ParentBoneSpace
    scaleMode=BMM_Additive
    translationMode=BMM_Additive
    alphaBoolBlend {blendOption=Linear}
    boneToModify {boneName=dyn_shoulder_pad_r}
    rotation {roll=-10}
   }
   Save cached pose 'Before_R_Forearm'(Pose=Component To Local(ComponentPose=@n24)) {cacheName=Before_R_Forearm}
  ]}
 EventGraph {nodes [
   Event BlueprintUpdateAnimation()
   n1 = GetOwnerSkeletalMeshComponent()
   n2 = BreakTransform(InTransform=GetDeltaTransformFromRefPose(self=@n1, BoneName=upperarm_r, BaseName=clavicle_r))
   > Set LookUp(LookUp:double=MapRangeClamped(Value=BreakRotator(InRot=@n2.Rotation).Yaw, InRangeA=95, InRangeB=120)) -> (Output_Get:double)
   n3 = BreakTransform(InTransform=GetDeltaTransformFromRefPose(self=@n1, BoneName=upperarm_l, BaseName=clavicle_l))
   > Set LookDown(LookDown:double=MapRangeClamped(Value=BreakRotator(InRot=@n3.Rotation).Yaw, InRangeA=70, InRangeB=20)) -> (Output_Get:double)
   n4 = GetDeltaTransformFromRefPose(self=@n1, BoneName=thigh_l, BaseName=calf_l)
   MapRangeClamped(Value=BreakRotator(InRot=BreakTransform(InTransform=@n4).Rotation).Yaw, InRangeA=40, InRangeB=-20)
   n5 = BreakRotator(InRot=BreakTransform(InTransform=GetDeltaTransformFromRefPose(self=@n1, BoneName=upperarm_r, BaseName=clavicle_r)).Rotation)
   "To String (Float)"(InDouble=@n5.Yaw)
   TryGetPawnOwner()
   MapRangeClamped(Value=@n5.Yaw, InRangeA=100, InRangeB=125)
  ]}
]
variables [
 LookDown {type=number}
 LookUp {type=number}
 NewVar_0 {type=number}
]
