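# Physics asset for the Dark Eagle Purple male medium player body, named as a cape collider, giving upper-body collision volumes that by its name serve cape simulation.
# Capsules cover the pelvis, spine, neck and clavicles, with a sphere on the head.
# The spine_02 and spine_05 bodies layer several capsules to fill out the torso.
# Arm, leg and foot bodies carry no shapes, and no constraints link the bodies.
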PhysicsAsset /BRCosmetics/Characters/Player/Male/Medium/Bodies/M_MED_Dark_Eagle_Purple/Meshes/M_MED_Dark_Eagle_Purple_CapeCollider
bodies {
 calf_l []
 calf_r []
 clavicle_l [{shapeType=Capsule, length=8.0136, radius=10.8927, center=<12.7625 -0.6512 3.0595>, rotation=<rot 75.9352 -0.0001 175.5622>}]
 clavicle_r [{shapeType=Capsule, length=8.0136, radius=10.8927, center=<-12.5682 0.6555 -2.515>, rotation=<rot 75.9351 -0.0001 -4.4378>}]
 foot_l []
 foot_r []
 head [{shapeType=Sphere, radius=17.3125, center=<11.028 -0.174 0>}]
 lowerarm_l []
 lowerarm_r []
 neck_02 [{shapeType=Capsule, length=13.4375, radius=7.3175, center=<-0.2642 -0.3408 0>, rotation=<rot 90 -171.1193 -175.3978>}]
 pelvis [{shapeType=Capsule, length=6.9062, radius=15.2888, center=<0.5472 0.362 0>}]
 spine_01 [{shapeType=Capsule, length=11.2187, radius=13.9138, center=<-2.6215 -0.1579 0>}]
 spine_02 [
  {shapeType=Capsule, length=12.8437, radius=11.445, center=<-2.6215 0.4993 0.0001>}
  {shapeType=Capsule, length=29.875, radius=11.1013, center=<2.5024 0.4992 3.9825>, rotation=<rot 90 0.0049 0.0049>}
  {shapeType=Capsule, length=29.875, radius=11.1013, center=<2.4704 0.4991 -3.9825>, rotation=<rot 90 -0.0396 -0.0396>}
 ]
 spine_03 [{shapeType=Capsule, length=9.4062, radius=10.5075, center=<-1.953 -1.3279 0.0001>}]
 spine_04 [{shapeType=Capsule, length=15.9062, radius=12.3825, center=<2.1467 -1.2748 0.0002>}]
 spine_05 [
  {shapeType=Capsule, length=21.8437, radius=13.195, center=<-1.7514 0.2923 0>}
  {shapeType=Capsule, length=16.2812, radius=13.5625, center=<0.7875 0.22 0>}
 ]
 thigh_l []
 thigh_r []
 upperarm_l []
 upperarm_r []
}
constraints []
